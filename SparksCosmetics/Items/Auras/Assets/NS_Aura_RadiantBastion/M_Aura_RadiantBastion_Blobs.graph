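# Lit material for the blob mesh particles in the Radiant Bastion aura's Niagara effect.
# A Fresnel Color rim, scaled by particle alpha, adds to tinted particle color and feeds an emissive adjustment function.
# A panning swirl normal, randomized per particle, distorts UVs by the dynamic WiggleIntensity.
# Separate branches build fresnel glow over a bounding-box-mapped lava mask and sample noise, bubble normal and mask textures, with no shown consumers.
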Material /SparksCosmetics/Items/Auras/Assets/NS_Aura_RadiantBastion/M_Aura_RadiantBastion_Blobs
graph {
 properties {
  bAutomaticallySetUsageInEditor=false
  bUsedWithNiagaraMeshParticles=true
  previewMesh=/Game/Effects/Fort_Effects/Gameplay/Water/Meshes/St_WaterDrop_Small_MorphExport_01_NoLOD
  usedShadingModels=Default Lit
 }
 nodes [
  n1 = Add(A=TextureCoordinate() {uTiling=0.4, vTiling=0.4}, B=ParticleRandom())
  n2 = ComponentMask(None=TextureSampleParameter2D(UVs=Panner(Coordinate=@n1) {speedY=0.4}) {parameterName=DistortTexture, samplerType=SAMPLERTYPE_Normal, texture=/Game/Global/Textures/Default/Noise/T_SwirlNoise_N}.RGB) {g=true, r=true}
  n3 = DynamicParameter() {parameterIndex=1, defaultValue=<rgba 1 1 0.2 0>, paramNames [Emissive, Erode, WiggleIntensity, Param4]}
  Clamp(None=Add(A=TextureCoordinate(), B=Multiply(A=@n2, B=@n3.WiggleIntensity)))
  n4 = Constant()
  n5 = VectorParameter() {parameterName=Fresnel Color, defaultValue=<rgba 0.8308 0.3515 1 1>}
  n6 = ParticleColor()
  n7 = Multiply(A=@n6.A, B=Fresnel() {exponent=1})
  n8 = Constant3Vector() {constant=<rgba 0.0038 0.1183 0.2 1>}
  MaterialFunctionCall(Emissive=Add(A=LinearInterpolate(A=@n4, B=@n5.RGB, Alpha=@n7), B=Multiply(A=@n6.RGB, B=@n8))) {materialFunction=&MF_FX_EmissiveAdjustments}
  n9 = TextureCoordinate()
  Multiply(A=TextureSample(UVs=Panner(Coordinate=@n9) {speedX=0.5, speedY=0.5}) {samplerType=SAMPLERTYPE_LinearGrayscale, texture=/Game/Effects/Fort_Effects/Emotes/UltraLight/Texture/T_Tiling_Noise_01_GreyScale}.G)
  n10 = VectorParameter() {parameterName=Glow Noise Color, defaultValue=<rgba 1 0.4893 0 1>}
  n11 = Fresnel() {baseReflectFraction=0.005}
  n12 = Panner(Coordinate=MaterialFunctionCall() {materialFunction=&BoundingBoxBased_0-1_UVW}.B) {speedX=0.1, speedY=0.3}
  Multiply(
   A=Add(
    A=Multiply(A=@n10.RGB, B=@n11)
    B=TextureSample(UVs=@n12) {samplerType=SAMPLERTYPE_Masks, texture=/Game/Environments/Sets/Vulcano/Textures/Lava_M}.R
   )
   B=@n8
  )
  ScalarParameter() {defaultValue=0.2, parameterName=WiggleIntensity}
  TextureSampleParameter2D(UVs=TextureCoordinate() {uTiling=2}) {parameterName=Normal, samplerType=SAMPLERTYPE_Normal, texture=/Game/Effects/Niagara/MasterMaterials/Liquid/T_Liquid_Bubbles_Normal}
  TextureSample() {samplerType=SAMPLERTYPE_Masks, texture=/SparksCosmetics/Items/Auras/Assets/NS_Aura_Magicstone/T_MagicStone_Aura_S}
 ]
}
