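# Translucent unlit Niagara sprite material that draws animated, noise-distorted ripples for the Radiant Bastion aura cosmetic.
# Radial distance plus rotating Perlin noise drives a time-animated sine for glow rings, alongside a wispy fire-noise pattern warped by panning Perlin, both under soft sphere masks.
# Glow and pattern each hue-shift the particle color and blend into emissive.
# Opacity combines particle-alpha-scaled pattern with glow, faded by a vertical gradient.
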Material /SparksCosmetics/Items/Auras/Assets/NS_Aura_RadiantBastion/M_Aura_RadiantBastion_Ripples
graph {
 properties {
  bAutomaticallySetUsageInEditor=false
  bUsedWithNiagaraSprites=true
  blendMode=BLEND_Translucent
  shadingModel=MSM_Unlit
  usedShadingModels=Unlit
 }
 nodes [
  n1 = NamedRerouteUsage()
  n2 = Multiply(A=ConstantBiasScale(None=NamedRerouteUsage()) {bias=-0.5, scale=2}) {constB=0.2}
  n3 = Rotator(Coordinate=TextureCoordinate()) {speed=0.2}
  n4 = Add(A=TextureCoordinate()) {constB=-0.5}
  n5 = Multiply(A=@n4, B=@n4)
  n6 = ComponentMask(None=@n5) {r=true}
  n7 = ComponentMask(None=@n5) {g=true}
  n8 = Add(
   A=Multiply(A=TextureSample(UVs=@n3) {samplerType=SAMPLERTYPE_LinearGrayscale, texture=&T_SimplePerlin_Greyscale}.R) {constB=0.15}
   B=SquareRoot(None=Add(A=@n6, B=@n7))
  )
  n9 = ScalarParameter() {defaultValue=-6, parameterName=Glow Tiling}
  n10 = Time()
  n11 = ScalarParameter() {defaultValue=1.5, parameterName=Glow Speed}
  n12 = Sine(None=Add(A=Multiply(A=@n8, B=@n9) {constB=-6}, B=Multiply(A=@n10, B=@n11) {constB=2}))
  n13 = Multiply(
   A=SphereMask(A=Add(A=@n1, B=@n2), B=Constant() {r=0.5}, Radius=Constant() {r=0.4}) {hardnessPercent=0}
   B=Add(A=Multiply(A=@n12) {constB=0.5}) {constB=0.5}
  )
  n14 = Multiply(A=@n13, B=ScalarParameter() {defaultValue=1, parameterName=Glow Hue Shift}) {constB=0.2}
  n15 = ParticleColor()
  n16 = TextureSample(UVs=Panner(Coordinate=TextureCoordinate()) {speedY=0.2}) {samplerType=SAMPLERTYPE_LinearGrayscale, texture=&T_SimplePerlin_Greyscale}
  n17 = Multiply(A=Time(), B=Constant() {r=0.2})
  n18 = Subtract(A=TextureSample(UVs=Add(A=@n16.R, B=@n17)) {texture=/Game/Effects/Fort_Effects/Textures/T_WispyFireNoise03}.R) {constB=0.36}
  n19 = Multiply(A=Power(Base=@n18), B=ScalarParameter() {defaultValue=1, parameterName=Pattern Hue Shift}) {constB=0.2}
  MaterialFunctionCall(Emissive=LinearInterpolate(
    A=MaterialFunctionCall(Hue Shift Percentage=@n14, Texture=@n15.RGB) {materialFunction=&HueShift}.Result
    B=MaterialFunctionCall(Hue Shift Percentage=@n19, Texture=@n15.RGB) {materialFunction=&HueShift}.Result
    Alpha=Constant() {r=0.25}
   )) {materialFunction=&MF_FX_EmissiveAdjustments}
  n20 = ConstantBiasScale(None=NamedRerouteUsage()) {bias=-0.5, scale=2}
  n21 = SphereMask(A=Add(A=TextureCoordinate(), B=Multiply(A=@n20) {constB=0.2}), B=Constant() {r=0.5}, Radius=Constant() {r=0.3}) {hardnessPercent=0}
  n22 = Multiply(A=@n15.A, B=Power(Base=Multiply(A=@n18, B=@n21)) {constExponent=1})
  n23 = NamedRerouteUsage()
  n24 = TextureCoordinate() {vTiling=1.55}
  Multiply(
   A=Saturate(None=Add(A=@n22, B=@n23))
   B=OneMinus(None=MaterialFunctionCall(UV Channel=@n24) {materialFunction=&LinearGradient}.VGradient)
  )
  NamedRerouteDeclaration(None=Multiply(A=Saturate(None=@n13), B=ScalarParameter() {defaultValue=0.8, parameterName=Glow Opacity}) {constB=0.1}) {name=GlowOpacity, nodeColor=<rgba 0 1 0.3906 1>}
  n25 = TextureCoordinate()
  n26 = Constant2Vector() {g=1, r=1}
  n27 = LinearInterpolate(Alpha=Power(Base=MaterialFunctionCall() {materialFunction=&LinearGradient}.VGradient)) {constA=1, constB=0.5}
  n28 = Constant() {r=1}
  NamedRerouteDeclaration(None=MaterialFunctionCall(UVs=Multiply(A=@n25, B=@n26), Texture Scale=AppendVector(A=@n27, B=@n28)) {materialFunction=&ScaleUVsByCenter}.UVs) {name=SphereMaskUVs, nodeColor=<rgba 1 0 0.0462 1>}
  NamedRerouteDeclaration(None=@n16.R) {name=Distortion R Channel, nodeColor=<rgba 0 1 0.6484 1>}
 ]
}
